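# One-shot stereo MetaSound stinger for a rank-up in the Cherry Hero Ranking System, apparently the first-person variant, routed to the boss music stinger sound class.
# On play, the Rank input indexes two wave arrays, and both picks play together through a stereo mixer and fixed gain boost.
# The first player ends the sound; the Play and bIsSPlus inputs are unconnected.
# Concurrency allows one instance per owner, and a level-up mute bus modulates volume.
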MetaSoundSource /CherrySmokeHeroRankingSystem/Assets/Audio/MSS_Ch6S3_CherryHeroRankingSystem_RankUp_1P
inputs [
 bIsSPlus {type=Bool}
 Play {type=Trigger}
 Rank {type=Int32}
 UE.Source.OnPlay {type=Trigger}
]
interfaces [UE.OutputFormat.Stereo:1.0, UE.Source.OneShot:1.0, UE.Source:1.1]
outputs [
 "UE.OutputFormat.Stereo.Audio:0" {type=Audio}
 "UE.OutputFormat.Stereo.Audio:1" {type=Audio}
 UE.Source.OneShot.OnFinished {type=Trigger}
]
pages [Default {
  input_defaults {Play=null, Rank=0, UE.Source.OnPlay=null, bIsSPlus=null}
  nodes [
   n1 = InitVariable.WaveAsset:Array()
   WavePlayer_0 = UE.Wave Player:Stereo(
    Play=^UE.Source.OnPlay
    Wave Asset=Array.Get:WaveAsset:Array(Index=^Rank, Trigger=^UE.Source.OnPlay, Array=VariableAccessor.WaveAsset:Array(Variable=@n1.Variable).Value).Element
   )
   n2 = InitVariable.WaveAsset:Array()
   WavePlayer_1 = UE.Wave Player:Stereo(
    Play=^UE.Source.OnPlay
    Wave Asset=Array.Get:WaveAsset:Array(Index=^Rank, Trigger=^UE.Source.OnPlay, Array=VariableAccessor.WaveAsset:Array(Variable=@n2.Variable).Value).Element
   )
   AudioMixerAudioMixerStereo2_0 = "AudioMixer.Audio Mixer (Stereo, 2)"(In 0 L=@WavePlayer_0.Out Left, In 0 R=@WavePlayer_0.Out Right, In 1 L=@WavePlayer_1.Out Left, In 1 R=@WavePlayer_1.Out Right)
   DecibelstoLinearGain_0 = UE.Decibels to Linear Gain:Float() {also {defaults {Decibels=2.5}}}
   MSP_StereoGain_0 = &MSP_StereoGain(
    Left In=@AudioMixerAudioMixerStereo2_0.Out L
    Right In=@AudioMixerAudioMixerStereo2_0.Out R
    "Gain (Linear)"=@DecibelstoLinearGain_0.Linear Gain
   )
  ]
  links [
   {dst="@UE.OutputFormat.Stereo.Audio:0", src=MSP_StereoGain_0.Left Out}
   {dst="@UE.OutputFormat.Stereo.Audio:1", src=MSP_StereoGain_0.Right Out}
   {dst="@UE.Source.OneShot.OnFinished", src=WavePlayer_0.On Finished}
  ]
 }]
properties {
 assetImportData=&self:AssetImportData
 bOverrideConcurrency=true
 numChannels=2
 outputFormat=Stereo
 sampleRate=48000
 soundClassObject=/Game/Sounds/FortSoundClassesAndModes/Fort_Music_Boss_Stinger
 volume=1.2
 concurrencyOverrides {bLimitToOwner=true, maxCount=1}
 modulationSettings {
  volumeRouting=Union
  volumeModulationDestination {bEnableModulation=true, modulators [&/CB_HeroRank_LevelUp_Mute]}
 }
}
