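# One-shot stereo MetaSound source, named as the third-person shield-heal rank-up sound for the Cherry hero ranking system, limited to one instance per owner.
# Mixes two random wave players, one gained by an ADSR envelope whose times scale from 70% of the Duration input.
# Envelope release input comes from its own release-triggered output through a deferred variable and delay.
# Finish requires both the ungated player and envelope to complete.
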MetaSoundSource /CherrySmokeHeroRankingSystem/Assets/Audio/MSS_Ch6S3_CherryHeroRankingSystem_RankUp_ShieldHeal_3P
inputs [
 Duration {type=Float}
 UE.Source.OnPlay {type=Trigger}
]
interfaces [UE.OutputFormat.Stereo:1.0, UE.Source.OneShot:1.0, UE.Source:1.1]
outputs [
 "UE.OutputFormat.Stereo.Audio:0" {type=Audio}
 "UE.OutputFormat.Stereo.Audio:1" {type=Audio}
 UE.Source.OneShot.OnFinished {type=Trigger}
]
pages [Default {
  input_defaults {Duration=16, UE.Source.OnPlay=null}
  nodes [
   n1 = /FortMetasound/Patches/MSP_Audio_StereoRng_Waveplayer(Play=^UE.Source.OnPlay, In Array=VariableAccessor.WaveAsset:Array(Variable=InitVariable.WaveAsset:Array().Variable).Value) {also {defaults {Loop="<MetaSoundBuilderResult.SUCCEEDED: 0>"}}}
   n2 = &MSP_Audio_StereoRng_Waveplayer(Play=^UE.Source.OnPlay, In Array=VariableAccessor.WaveAsset:Array(Variable=InitVariable.WaveAsset:Array().Variable).Value)
   DecibelstoLinearGain_1 = UE.Decibels to Linear Gain:Float() {also {defaults {Decibels=-6}}}
   n3 = InitVariable.Float()
   n4 = UE.Subtract:Float(PrimaryOperand=^Duration, AdditionalOperands=UE.Multiply:Float(PrimaryOperand=^Duration) {also {defaults {AdditionalOperands=0.3}}}.Out)
   n5 = UE.Multiply:Float(
    AdditionalOperands=UE.Divide:Float() {also {defaults {AdditionalOperands=8}}}.Out
    PrimaryOperand=VariableAccessor.Float(Variable=VariableMutator.Float(Variable=@n3.Variable, Value=@n4.Out).Variable).Value
   )
   n6 = UE.ConversionFloatToTime(In=UE.Multiply:Float(PrimaryOperand=@n5.Out) {also {defaults {AdditionalOperands=0.5}}}.Out)
   n7 = InitVariable.Time()
   n8 = VariableDeferredAccessor.Trigger(Variable=InitVariable.Trigger().Variable)
   TriggerDelay_0 = UE.Trigger Delay(Delay Time=VariableAccessor.Time(Variable=VariableMutator.Time(Value=@n6.Out, Variable=@n7.Variable).Variable).Value, In=@n8.Value)
   n9 = UE.ConversionFloatToTime(In=UE.Multiply:Float(PrimaryOperand=@n5.Out) {also {defaults {AdditionalOperands=3}}}.Out)
   n10 = InitVariable.Time()
   n11 = UE.ConversionFloatToTime(In=@n5.Out)
   n12 = InitVariable.Time()
   n13 = UE.ConversionFloatToTime(In=UE.Multiply:Float(PrimaryOperand=@n5.Out) {also {defaults {AdditionalOperands=4.5}}}.Out)
   n14 = InitVariable.Time()
   ADSREnvelopeADSREnvelope_0 = ADSR Envelope.ADSR Envelope:Float(
    Trigger Attack=^UE.Source.OnPlay
    Sustain Level=@DecibelstoLinearGain_1.Linear Gain
    Trigger Release=@TriggerDelay_0.Out
    Decay Time=VariableAccessor.Time(Variable=VariableMutator.Time(Value=@n9.Out, Variable=@n10.Variable).Variable).Value
    Attack Time=VariableAccessor.Time(Variable=VariableMutator.Time(Value=@n11.Out, Variable=@n12.Variable).Variable).Value
    Release Time=VariableAccessor.Time(Variable=VariableMutator.Time(Value=@n13.Out, Variable=@n14.Variable).Variable).Value
   ) {also {defaults {Attack Curve=0.5, Hard Reset="<MetaSoundBuilderResult.SUCCEEDED: 0>", Release Curve=0.5}}}
   AudioMixerAudioMixerStereo2_0 = "AudioMixer.Audio Mixer (Stereo, 2)"(
    In 1 L=@n1.Out.L
    In 1 R=@n1.Out.R
    In 0 L=@n2.Out.L
    In 0 R=@n2.Out.R
    Gain 1=UE.Multiply:Float(AdditionalOperands=@ADSREnvelopeADSREnvelope_0.Out Envelope) {also {defaults {PrimaryOperand=0.4}}}.Out
   )
   DecibelstoLinearGain_0 = UE.Decibels to Linear Gain:Float() {also {defaults {Decibels=-3}}}
   Multiply_0 = UE.Multiply:Audio by Float(PrimaryOperand=@AudioMixerAudioMixerStereo2_0.Out R, AdditionalOperands=@DecibelstoLinearGain_0.Linear Gain)
   Multiply_1 = UE.Multiply:Audio by Float(PrimaryOperand=@AudioMixerAudioMixerStereo2_0.Out L, AdditionalOperands=@DecibelstoLinearGain_0.Linear Gain)
   TriggerAccumulatorTriggerAccumulate2_0 = "TriggerAccumulator.Trigger Accumulate (2)"(In 1=@ADSREnvelopeADSREnvelope_0.On Done, In 0=@n2.On Finished)
   VariableMutator.Trigger(Value=@ADSREnvelopeADSREnvelope_0.On Release Triggered, Variable=@n8.Variable)
  ]
  links [
   {dst="@UE.OutputFormat.Stereo.Audio:1", src=Multiply_0.Out}
   {dst="@UE.OutputFormat.Stereo.Audio:0", src=Multiply_1.Out}
   {dst="@UE.Source.OneShot.OnFinished", src=TriggerAccumulatorTriggerAccumulate2_0.Out}
  ]
 }]
properties {
 assetImportData=&self:AssetImportData
 bOverrideConcurrency=true
 numChannels=2
 outputFormat=Stereo
 sampleRate=48000
 soundClassObject=&Fort_Consumables
 concurrencyOverrides {bLimitToOwner=true, maxCount=1}
 modulationSettings {volumeModulationDestination {bEnableModulation=true}}
}
