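# One-shot stereo MetaSound Source for the XP notification sound in the Cherry Smoke hero ranking system, classed as non-critical in-game UI audio.
# On play, a Wave Player fires the XP point marker wave into a stereo gain patch driven by a fixed decibel-to-linear boost.
# The player's finish trigger ends the source.
# Overridden concurrency caps simultaneous copies per owner, and volume modulation is enabled.
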MetaSoundSource /CherrySmokeHeroRankingSystem/Assets/Audio/MSS_Ch6S3_CherryHeroRankingSystem_XpNotify
inputs [UE.Source.OnPlay {type=Trigger}]
interfaces [UE.OutputFormat.Stereo:1.0, UE.Source.OneShot:1.0, UE.Source:1.1]
outputs [
 "UE.OutputFormat.Stereo.Audio:0" {type=Audio}
 "UE.OutputFormat.Stereo.Audio:1" {type=Audio}
 UE.Source.OneShot.OnFinished {type=Trigger}
]
pages [Default {
  input_defaults {UE.Source.OnPlay=null}
  nodes [
   DecibelstoLinearGain_0 = UE.Decibels to Linear Gain:Float() {also {defaults {Decibels=2}}}
   WavePlayer_0 = UE.Wave Player:Stereo(Play=^UE.Source.OnPlay) {also {defaults {Wave Asset=&/CherrySmokeHeroRank_XpPointMarker}}}
   MSP_StereoGain_0 = &MSP_StereoGain("Gain (Linear)"=@DecibelstoLinearGain_0.Linear Gain, Left In=@WavePlayer_0.Out Left, Right In=@WavePlayer_0.Out Right)
  ]
  links [
   {dst="@UE.OutputFormat.Stereo.Audio:0", src=MSP_StereoGain_0.Left Out}
   {dst="@UE.OutputFormat.Stereo.Audio:1", src=MSP_StereoGain_0.Right Out}
   {dst="@UE.Source.OneShot.OnFinished", src=WavePlayer_0.On Finished}
  ]
 }]
properties {
 assetImportData=&self:AssetImportData
 bOverrideConcurrency=true
 numChannels=2
 outputFormat=Stereo
 sampleRate=48000
 soundClassObject=&Fort_UI_InGame_NonCritical
 concurrencyOverrides {bLimitToOwner=true, maxCount=2}
 modulationSettings {volumeModulationDestination {bEnableModulation=true}}
}
